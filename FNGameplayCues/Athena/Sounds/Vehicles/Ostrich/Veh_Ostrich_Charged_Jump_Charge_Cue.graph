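# Sound cue for the charge-up of the Ostrich mech vehicle's charged jump.
# A single wave player feeds the output at reduced volume through the enemy third-person player foley sound class.
# It references the Ostrich movement ability attenuation asset and carries override values for long-range falloff, distance low-pass filtering and reverb send.
# Overridden concurrency allows one instance per owner, with a new play stopping the oldest.
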SoundCue /FNGameplayCues/Athena/Sounds/Vehicles/Ostrich/Veh_Ostrich_Charged_Jump_Charge_Cue
root=WavePlayer_0
nodes [{id=WavePlayer_0, properties {soundWaveAssetPtr=&/Mech_Vehicle_ChargingJump_ChargeUp_01}}]
properties {
 attenuationSettings=/Game/Sounds/Attenuation/Fort_Ostrich_Movement_Ability_Att
 bOverrideConcurrency=true
 maxDistance=15600
 soundClassObject=&Fort_Player_Foley_3P_Enemy_Athena
 volumeMultiplier=0.6
 attenuationOverrides {
  bAttenuateWithLPF=true
  falloffDistance=15000
  lPFFrequencyAtMax=2000
  lPFRadiusMax=7500
  lPFRadiusMin=1000
  reverbWetLevelMax=0.7
  reverbWetLevelMin=0.2
  attenuationShapeExtents {x=1000}
 }
 concurrencyOverrides {bLimitToOwner=true, maxCount=1, resolutionRule=StopOldest}
}
