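# Sound cue for the Ostrich vehicle's jump ability, varying with charge state.
# A switch on the IsFullyCharged int parameter picks randomized Dash waves for 0 or Charged_Jump waves for 1, with silence when the parameter is unset.
# Distance attenuation applies, and a per-owner concurrency override allows one instance, stopping the oldest.
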SoundCue /FNGameplayCues/Athena/Sounds/Vehicles/Ostrich/Veh_Ostrich_Charged_Jump_Cue
root=Switch_0
nodes [
 {id=Random_0, children [WavePlayer_0,WavePlayer_1], properties {weights [1,1]}}
 {id=Random_1, children [WavePlayer_2,WavePlayer_3,WavePlayer_4], properties {weights [1,1,1]}}
 {id=Switch_0, children [null,Random_1,Random_0], properties {intParameterName=IsFullyCharged}}
 {id=WavePlayer_0, properties {soundWaveAssetPtr=&/Veh_Ostrich_Charged_Jump_01}}
 {id=WavePlayer_1, properties {soundWaveAssetPtr=&/Veh_Ostrich_Charged_Jump_02}}
 {id=WavePlayer_2, properties {soundWaveAssetPtr=&/Veh_Ostrich_Dash_01}}
 {id=WavePlayer_3, properties {soundWaveAssetPtr=&/Veh_Ostrich_Dash_02}}
 {id=WavePlayer_4, properties {soundWaveAssetPtr=&/Veh_Ostrich_Dash_03}}
]
properties {
 attenuationSettings=/Game/Sounds/Attenuation/Fort_Ostrich_Movement_Ability_Att
 bOverrideConcurrency=true
 maxDistance=15600
 soundClassObject=&Fort_Player_Foley_3P_Enemy_Athena
 volumeMultiplier=0.5
 attenuationOverrides {falloffDistance=5000, reverbWetLevelMax=0.7, reverbWetLevelMin=0.2, attenuationShapeExtents {x=1000}}
 concurrencyOverrides {bLimitToOwner=true, maxCount=1, resolutionRule=StopOldest, retriggerTime=0.5}
}
