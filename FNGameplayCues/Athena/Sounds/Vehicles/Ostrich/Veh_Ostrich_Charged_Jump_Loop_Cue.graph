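# Sound cue for the looping charged-jump sound of the Ostrich vehicle.
# A fixed two-second Delay node at the root holds back a looping Wave Player that plays a single charge loop wave.
# It plays at reduced volume through the third-person enemy player foley sound class, uses the shared Ostrich close-movement attenuation, and overrides concurrency to prevent new instances once the limit is reached.
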SoundCue /FNGameplayCues/Athena/Sounds/Vehicles/Ostrich/Veh_Ostrich_Charged_Jump_Loop_Cue
root=Delay_0
nodes [
 {id=Delay_0, children [WavePlayer_0], properties {delayMax=2, delayMin=2}}
 {
  id=WavePlayer_0
  properties {bLooping=true, soundWaveAssetPtr=&/Veh_Ostrich_Charged_Jump_Loop_01}
 }
]
properties {
 attenuationSettings=/Game/Sounds/Attenuation/Fort_Ostrich_Movement_Close_Att
 bOverrideConcurrency=true
 maxDistance=6512
 soundClassObject=&Fort_Player_Foley_3P_Enemy_Athena
 volumeMultiplier=0.5
 attenuationOverrides {falloffDistance=5000, reverbWetLevelMax=0.7, reverbWetLevelMin=0.2, attenuationShapeExtents {x=1000}}
 concurrencyOverrides {resolutionRule=PreventNew}
}
