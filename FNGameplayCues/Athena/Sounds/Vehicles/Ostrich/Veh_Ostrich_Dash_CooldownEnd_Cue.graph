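# Sound cue for the Ostrich vehicle that, going by its name, signals the end of the dash cooldown with a jets-ready sound.
# A single wave player feeds the output directly at reduced volume, with no randomization or mixing.
# It uses close-range Ostrich movement attenuation and a 2D hit-notification instigator sound class.
# It references the local footstep concurrency set and carries per-owner, single-voice, stop-oldest override settings.
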SoundCue /FNGameplayCues/Athena/Sounds/Vehicles/Ostrich/Veh_Ostrich_Dash_CooldownEnd_Cue
root=WavePlayer_0
nodes [{id=WavePlayer_0, properties {soundWaveAssetPtr=&/Veh_Ostrich_Dash_Jets_Ready_01}}]
properties {
 attenuationSettings=/Game/Sounds/Attenuation/Fort_Ostrich_Movement_Close_Att
 maxDistance=6512
 soundClassObject=&Fort_2DHitNotifications_Instigate_PSM
 volumeMultiplier=0.5
 concurrencyOverrides {bLimitToOwner=true, maxCount=1, resolutionRule=StopOldest}
 concurrencySet [&Athena_Footstep_Local_Conc]
}
