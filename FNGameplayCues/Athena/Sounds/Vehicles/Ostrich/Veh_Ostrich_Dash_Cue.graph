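# Sound cue for the Ostrich vehicle's dash.
# A mixer layers two branches, each randomly picking one of three equally weighted waves through its own modulator: a dash sound and an Ostrich FS (footstep) sound, with the FS layer mixed louder.
# The cue plays at reduced volume, using distant Ostrich movement attenuation and the third-person enemy player foley sound class.
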SoundCue /FNGameplayCues/Athena/Sounds/Vehicles/Ostrich/Veh_Ostrich_Dash_Cue
root=Mixer_2
nodes [
 {id=Mixer_2, children [Modulator_0,Modulator_1], properties {inputVolume [1,1.5]}}
 {id=Modulator_0, children [Random_1]}
 {id=Modulator_1, children [Random_0]}
 {id=Random_0, children [WavePlayer_3,WavePlayer_4,WavePlayer_5], properties {weights [1,1,1]}}
 {id=Random_1, children [WavePlayer_0,WavePlayer_1,WavePlayer_2], properties {weights [1,1,1]}}
 {id=WavePlayer_0, properties {soundWaveAssetPtr=&/Veh_Ostrich_Dash_01}}
 {id=WavePlayer_1, properties {soundWaveAssetPtr=&/Veh_Ostrich_Dash_02}}
 {id=WavePlayer_2, properties {soundWaveAssetPtr=&/Veh_Ostrich_Dash_03}}
 {id=WavePlayer_3, properties {soundWaveAssetPtr=/Game/Athena/Sounds/Vehicles/Ostrich/Veh_Ostrich_FS_03}}
 {id=WavePlayer_4, properties {soundWaveAssetPtr=/Game/Athena/Sounds/Vehicles/Ostrich/Veh_Ostrich_FS_01}}
 {id=WavePlayer_5, properties {soundWaveAssetPtr=/Game/Athena/Sounds/Vehicles/Ostrich/Veh_Ostrich_FS_02}}
]
properties {
 attenuationSettings=/Game/Sounds/Attenuation/Fort_Ostrich_Movement_Distant_Att
 maxDistance=15000
 soundClassObject=&Fort_Player_Foley_3P_Enemy_Athena
 volumeMultiplier=0.5
}
