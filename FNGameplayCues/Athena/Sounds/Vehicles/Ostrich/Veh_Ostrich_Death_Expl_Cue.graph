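# Sound cue for the Ostrich vehicle death explosion.
# A random node picks between two explosion waves with equal weight.
# It uses an Ostrich death attenuation asset, a third-person ranged weapon fire sound class and a boosted volume.
# A concurrency override caps simultaneous instances and stops the oldest when the limit is reached.
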SoundCue /FNGameplayCues/Athena/Sounds/Vehicles/Ostrich/Veh_Ostrich_Death_Expl_Cue
root=Random_0
nodes [
 {id=Random_0, children [WavePlayer_3,WavePlayer_4], properties {weights [1,1]}}
 {id=WavePlayer_3, properties {soundWaveAssetPtr=&/Veh_Ostrich_Death_Expl_01}}
 {id=WavePlayer_4, properties {soundWaveAssetPtr=&/Veh_Ostrich_Death_Expl_02}}
]
properties {
 attenuationSettings=/FNGameplayCues/Sounds/Attenuation/Fort_Veh_Ostrich_Death_Att
 bOverrideConcurrency=true
 maxDistance=15600
 soundClassObject=&Fort_WeaponRanged_Fire_3P_PSM
 volumeMultiplier=1.5
 concurrencyOverrides {maxCount=3, resolutionRule=StopOldest}
}
