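# Sound cue for the charge-up sound of the Ostrich vehicle's self-destruct death.
# A single wave player plays the charge-up wave directly, with no randomization or modulation nodes.
# It uses the Ostrich death attenuation asset and a third-person ranged weapon fire sound class at reduced volume.
# A concurrency override caps simultaneous instances and stops the oldest when the limit is hit.
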SoundCue /FNGameplayCues/Athena/Sounds/Vehicles/Ostrich/Veh_Ostrich_Death_Self_Destruct_ChargeUp_Cue
root=WavePlayer_2
nodes [{id=WavePlayer_2, properties {soundWaveAssetPtr=&/Veh_Ostrich_Death_Charge_Up_02}}]
properties {
 attenuationSettings=/FNGameplayCues/Sounds/Attenuation/Fort_Veh_Ostrich_Death_Att
 bOverrideConcurrency=true
 maxDistance=15600
 soundClassObject=&Fort_WeaponRanged_Fire_3P_PSM
 volumeMultiplier=0.5
 concurrencyOverrides {maxCount=3, resolutionRule=StopOldest}
}
