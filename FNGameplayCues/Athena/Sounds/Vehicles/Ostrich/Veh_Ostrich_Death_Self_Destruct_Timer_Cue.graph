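# Sound cue for the Ostrich vehicle's death self-destruct timer in Athena.
# A single wave player feeds one timer recording straight to the output, with no randomization or modulation.
# It plays in the Fort_Vehicles sound class at slightly reduced volume, with spatialization from the shared Ostrich death attenuation asset.
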SoundCue /FNGameplayCues/Athena/Sounds/Vehicles/Ostrich/Veh_Ostrich_Death_Self_Destruct_Timer_Cue
root=WavePlayer_1
nodes [{id=WavePlayer_1, properties {soundWaveAssetPtr=&/Veh_Ostrich_Death_Self_Destruct_Timer_01}}]
properties {
 attenuationSettings=/FNGameplayCues/Sounds/Attenuation/Fort_Veh_Ostrich_Death_Att
 maxDistance=15600
 soundClassObject=&Fort_Vehicles
 volumeMultiplier=0.8571
 attenuationOverrides {falloffDistance=1000, attenuationShapeExtents {x=1000}}
}
