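# Sound cue for the Ostrich vehicle's death warning.
# It plays a single warning wave after a fixed short delay through the vehicle sound class with a dedicated death attenuation asset.
# It overrides concurrency to allow few simultaneous instances, applying volume scaling with an attack time.
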SoundCue /FNGameplayCues/Athena/Sounds/Vehicles/Ostrich/Veh_Ostrich_Death_Warning_Cue
root=Delay_0
nodes [
 {id=Delay_0, children [WavePlayer_1], properties {delayMax=0.75, delayMin=0.75}}
 {id=WavePlayer_1, properties {soundWaveAssetPtr=&/Veh_Ostrich_Death_Warning_01}}
]
properties {
 attenuationSettings=/FNGameplayCues/Sounds/Attenuation/Fort_Veh_Ostrich_Death_Att
 bOverrideConcurrency=true
 maxDistance=15600
 soundClassObject=&Fort_Vehicles
 volumeMultiplier=0.8571
 attenuationOverrides {falloffDistance=1000, attenuationShapeExtents {x=1000}}
 concurrencyOverrides {maxCount=2, volumeScale=0.7, volumeScaleAttackTime=0.5}
}
