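# Sound cue for the Ostrich vehicle's default heavy landing footstep, using the third-person enemy player foley sound class.
# A mixer layers two attenuated branches: a random pick of two heavy-landing waves through a modulator with distant attenuation, and a random pick of five reused explosion dirt-and-stone debris waves with close attenuation.
# The cue itself references the distant attenuation asset, and its concurrency override fields specify one instance per owner, stopping the oldest.
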SoundCue /FNGameplayCues/Athena/Sounds/Vehicles/Ostrich/Veh_Ostrich_FS_Default_Land_Heavy_Bank_Cue
root=Mixer_0
nodes [
 {id=Attenuation_0, children [Random_1], properties {attenuationSettings=/Game/Sounds/Attenuation/Fort_Ostrich_Movement_Close_Att}}
 {id=Attenuation_1, children [Modulator_1], properties {attenuationSettings=/Game/Sounds/Attenuation/Fort_Ostrich_Movement_Distant_Att}}
 {id=Mixer_0, children [Attenuation_1, Attenuation_0], properties {inputVolume [1,1]}}
 {id=Modulator_1, children [Random_0]}
 {id=Random_0, children [WavePlayer_2,WavePlayer_3], properties {weights [1,1]}}
 {id=Random_1, children [WavePlayer_0,WavePlayer_1,WavePlayer_4,WavePlayer_5,WavePlayer_6], properties {weights [1,1,1,1,1]}}
 {id=WavePlayer_0, properties {soundWaveAssetPtr=/Game/Athena/Sounds/Weapons/Explosions/EXPL_Debris_Dirt_Stone_01}}
 {id=WavePlayer_1, properties {soundWaveAssetPtr=/Game/Athena/Sounds/Weapons/Explosions/EXPL_Debris_Dirt_Stone_03}}
 {id=WavePlayer_2, properties {soundWaveAssetPtr=/Game/Athena/Sounds/Vehicles/Ostrich/Veh_Ostrich_FS_Land_Heavy_01}}
 {id=WavePlayer_3, properties {soundWaveAssetPtr=/Game/Athena/Sounds/Vehicles/Ostrich/Veh_Ostrich_FS_Land_Heavy_02}}
 {id=WavePlayer_4, properties {soundWaveAssetPtr=/Game/Athena/Sounds/Weapons/Explosions/EXPL_Debris_Dirt_Stone_05}}
 {id=WavePlayer_5, properties {soundWaveAssetPtr=/Game/Athena/Sounds/Weapons/Explosions/EXPL_Debris_Dirt_Stone_02}}
 {id=WavePlayer_6, properties {soundWaveAssetPtr=/Game/Athena/Sounds/Weapons/Explosions/EXPL_Debris_Dirt_Stone_04}}
]
properties {
 attenuationSettings=&Fort_Ostrich_Movement_Distant_Att
 maxDistance=15000
 soundClassObject=&Fort_Player_Foley_3P_Enemy_Athena
 volumeMultiplier=0.5
 attenuationOverrides {bAttenuate=false, falloffDistance=30000, attenuationShapeExtents {x=0}}
 concurrencyOverrides {bLimitToOwner=true, maxCount=1, resolutionRule=StopOldest}
 concurrencySet [null]
}
