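# Sound cue for the Ostrich vehicle's footstep as it exits water in Athena, on the player foley sound class.
# A random node picks one of four equally weighted water-exit waves, then a modulator varies pitch and volume before a mixer whose second input is empty.
# Playback uses the Ostrich distant movement attenuation and the local footstep concurrency set, at a reduced overall volume.
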SoundCue /FNGameplayCues/Athena/Sounds/Vehicles/Ostrich/Veh_Ostrich_FS_Water_Exit_Cue
root=Mixer_0
nodes [
 {id=Mixer_0, children [Modulator_0,null], properties {inputVolume [1,1]}}
 {id=Modulator_0, children [Random_0], properties {pitchMin=0.9, volumeMin=0.9}}
 {id=Random_0, children [WavePlayer_1,WavePlayer_2,WavePlayer_3,WavePlayer_4], properties {weights [1,1,1,1]}}
 {id=WavePlayer_1, properties {soundWaveAssetPtr=&/Veh_Ostrich_FS_Water_Exit_01}}
 {id=WavePlayer_2, properties {soundWaveAssetPtr=&/Veh_Ostrich_FS_Water_Exit_02}}
 {id=WavePlayer_3, properties {soundWaveAssetPtr=&/Veh_Ostrich_FS_Water_Exit_03}}
 {id=WavePlayer_4, properties {soundWaveAssetPtr=&/Veh_Ostrich_FS_Water_Exit_04}}
]
properties {
 attenuationSettings=/Game/Sounds/Attenuation/Fort_Ostrich_Movement_Distant_Att
 maxDistance=15000
 soundClassObject=&Fort_Player_Foley_Athena
 volumeMultiplier=0.7
 concurrencyOverrides {bLimitToOwner=true, maxCount=1, resolutionRule=StopOldest}
 concurrencySet [&Athena_Footstep_Local_Conc]
}
