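# Sound cue for the Ostrich mech vehicle's jump slam in Athena.
# A single wave player plays one stomp air-lunge sample, with no randomization, mixing or modulation.
# It uses the shared distant Ostrich movement attenuation asset and is routed to a melee weapon sound class.
# Its own concurrency override prevents new instances from starting once the limit is reached.
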SoundCue /FNGameplayCues/Athena/Sounds/Vehicles/Ostrich/Veh_Ostrich_JumpSlam_Cue
root=WavePlayer_5
nodes [{id=WavePlayer_5, properties {soundWaveAssetPtr=&/Mech_Vehicle_Stomp_Air_Lunge_01}}]
properties {
 attenuationSettings=/Game/Sounds/Attenuation/Fort_Ostrich_Movement_Distant_Att
 bOverrideConcurrency=true
 maxDistance=15000
 soundClassObject=&Fort_WeaponMelee_PSM
 volumeMultiplier=1
 attenuationOverrides {falloffDistance=5000, reverbWetLevelMax=0.7, reverbWetLevelMin=0.2, attenuationShapeExtents {x=1000}}
 concurrencyOverrides {resolutionRule=PreventNew}
}
