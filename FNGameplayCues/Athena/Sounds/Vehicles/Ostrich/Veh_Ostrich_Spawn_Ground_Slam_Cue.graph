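# Sound cue for the ground slam when the Ostrich vehicle spawns, built from a close and a distant layer.
# A mixer blends the two single-wave branches, weighting the close one higher, with each passing through its own sound class, modulator and attenuation.
# Sound class overrides borrow ranged-weapon fire classes (third-person and distant), and modulators add random variation.
# The cue uses the shared weapon explosion concurrency rule.
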SoundCue /FNGameplayCues/Athena/Sounds/Vehicles/Ostrich/Veh_Ostrich_Spawn_Ground_Slam_Cue
root=Mixer_4
nodes [
 {
  id=Attenuation_0
  children [Modulator_1]
  properties {
   attenuationSettings=/FNGameplayCues/Sounds/Attenuation/Fort_Veh_Ostrich_SpawnLoop_Att
   attenuationOverrides {falloffDistance=20000, reverbWetLevelMax=0.25, reverbWetLevelMin=0.1, stereoSpread=1000, attenuationShapeExtents {x=2000}}
  }
 }
 {
  id=Attenuation_4
  children [WavePlayer_0]
  properties {
   attenuationSettings=/Game/Sounds/Attenuation/Fort_Ostrich_Movement_Close_Att
   bOverrideAttenuation=true
   attenuationOverrides {falloffDistance=10000, attenuationShapeExtents {x=500}}
  }
 }
 {id=Mixer_4, children [SoundClass_3,SoundClass_0], properties {inputVolume [1.25,0.75]}}
 {id=Modulator_0, children [Attenuation_4]}
 {id=Modulator_1, children [WavePlayer_4], properties {pitchMax=1.1, pitchMin=0.9}}
 {
  id=SoundClass_0
  children [Attenuation_0]
  properties {soundClassOverride=&Fort_WeaponRanged_Fire_Distant}
 }
 {id=SoundClass_3, children [Modulator_0], properties {soundClassOverride=&Fort_WeaponRanged_Fire_3P_PSM}}
 {id=WavePlayer_0, properties {soundWaveAssetPtr=&/Veh_Ostrich_Spawn_GroundSlam_01}}
 {id=WavePlayer_4, properties {soundWaveAssetPtr=&/Veh_Ostrich_Spawn_GroundSlam_Dist_01}}
]
properties {
 maxDistance=1125000
 soundClassObject=&Fort_WeaponRanged_Fire_3P_PSM
 volumeMultiplier=1
 concurrencyOverrides {maxCount=4, resolutionRule=StopOldest, volumeScale=0.5}
 concurrencySet [/Game/Sounds/Concurrency_Rules/Weapon_Explosion_Conc]
}
