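# Sound cue for the sonic boom that plays when the Ostrich vehicle spawns.
# A single wave player feeds a modulator that randomly varies pitch slightly at or below normal.
# It uses a dedicated Ostrich spawn attenuation asset with a very large max distance, and routes through a third-person ranged-weapon fire sound class rather than a vehicle-specific one.
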SoundCue /FNGameplayCues/Athena/Sounds/Vehicles/Ostrich/Veh_Ostrich_Spawn_SonicBoom_Cue
root=Modulator_0
nodes [
 {id=Modulator_0, children [WavePlayer_1], properties {pitchMax=1, pitchMin=0.9}}
 {id=WavePlayer_1, properties {soundWaveAssetPtr=&/Veh_Ostrich_Spawn_SonicBoom_01}}
]
properties {
 attenuationSettings=/FNGameplayCues/Sounds/Attenuation/Fort_Veh_Ostrich_Spawn_Att
 maxDistance=1125000
 soundClassObject=&Fort_WeaponRanged_Fire_3P_PSM
 volumeMultiplier=1
}
